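; Dojo prop Blueprint, apparently a five-jar stack, that builds its jars as hierarchical instanced static meshes.
; The construction script calls its parent, adds one instance per Instance Transforms entry, then rotates each.
; A random stream seeded from actor location gives repeatable yaw within plus or minus InstanceRotationAmount; existing pitch and roll are also combined in again.
; Destroying the actor destroys its Box component; other events are empty.
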(event EventBeginPlay)

(event Collision|EventActorBeginOverlap
  (params (OtherActor :type Actor%20Object%20Reference)))

(event EventTick
  (params (DeltaSeconds :type Float%20%28single-precision%29)))

(event EventDestroyed
  (Components|DestroyComponent (Variables|Default|GetBox)))

(fn GetInstanceCount
  (Variables|Default|SetInstanceCount (Components|InstancedStaticMesh|GetInstanceCount (Variables|Default|GetISM_Component))))

(fn SpawnInstances
  (for _array_element (Variables|Default|GetInstance%20Transforms)
    (Components|InstancedStaticMesh|AddInstance (Variables|Default|GetISM_Component) _array_element false)))

(fn UpdateInstanceRotation
  (Variables|Default|SetRandomStream (Math|Random|MakeRandomStreamfromLocation (Transformation|GetActorLocation) 1.0 true))
  (for _index (range (Variables|Default|GetInstanceCount))
    (Components|InstancedStaticMesh|UpdateInstanceTransform (Variables|Default|GetISM_Component) _index (Math|Transform|MakeTransform (.location (Components|InstancedStaticMesh|GetInstanceTransform (Variables|Default|GetISM_Component) _index false)) (Math|Rotator|CombineRotators (.rotation (Components|InstancedStaticMesh|GetInstanceTransform (Variables|Default|GetISM_Component) _index false)) (Math|Rotator|MakeRotator (.roll (.rotation (Components|InstancedStaticMesh|GetInstanceTransform (Variables|Default|GetISM_Component) _index false))) (.pitch (.rotation (Components|InstancedStaticMesh|GetInstanceTransform (Variables|Default|GetISM_Component) _index false))) (Math|Random|RandomFloatinRangefromStream (Variables|Default|GetRandomStream) (- (Variables|Default|GetInstanceRotationAmount) (* (Variables|Default|GetInstanceRotationAmount) 2.0)) (Variables|Default|GetInstanceRotationAmount)))) (.scale (Components|InstancedStaticMesh|GetInstanceTransform (Variables|Default|GetISM_Component) _index false))) false false false)))

(fn ConstructionScript
  (|CallParentFunction)
  (bind _returnvalue (AddComponent|NODE_AddHierarchicalInstancedStaticMeshComponent-1))
  (Variables|Default|SetISM_Component _returnvalue)
  (CallFunction|SpawnInstances)
  (CallFunction|GetInstanceCount)
  (CallFunction|UpdateInstanceRotation))
